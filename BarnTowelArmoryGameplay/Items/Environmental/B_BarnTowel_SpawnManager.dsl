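; Landing-pad manager that spawns a BarnTowel or MurkyVole armory skydrop pod overhead and manages its landing.
; Pod state changes drive World Partition data layer swaps, an authority-only GameplayEffect knockback capsule, a marker, and events to registry-tagged reactive prop components.
; On landing, the variant's interior layer activates; the server destroys the pod, while clients poll streaming completion before hiding it to avoid popping.
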(event EventBeginPlay
  (bind _landingknockbackcapsule (Variables|Default|GetLandingKnockbackCapsule))
  (bind _landingknockbackcapsule_2 (Variables|Default|GetLandingKnockbackCapsule))
  ;[0.002384 1 0] BeginPlay
  ; Setup registration using FortAutoRegister components
  (|CallParentFunction)
  (Networking|SwitchHasAuthority
    (:Authority
      (CallFunction|GetandRegisterReactiveArmoryProps))
    (:Remote
      _))
  (Components|Capsule|SetCapsuleRadius _landingknockbackcapsule (ScalableFloat|GetValueatLevel (Variables|Landing|Knockback|GetHFRow_KnockbackRadius)) true)
  (Components|Capsule|SetCapsuleHalfHeight _landingknockbackcapsule (ScalableFloat|GetValueatLevel (Variables|Default|GetHFRow_KnockbackHeight)) true)
  (Variables|/Script/Engine.PrimitiveComponent|SetbGenerateOverlapEvents false _landingknockbackcapsule_2)
  (Collision|SetCollisionEnabled _landingknockbackcapsule_2 "NoCollision"))

(event Custom|DebugSpawnArmory
  ;[0.5 0.5 0.5] DebugSpawnArmory
  ; Formerly used as a cheat, presevered in case people are still using it for tessting
  (CallFunction|SpawnArmoryinAir (Utilities|Select :Index (Math|Random|RandomInteger 1) :Option%200 "NewEnumerator2" :Option%201 "NewEnumerator1")))

(event Custom|LoadDataLayers
  (bind (_datatable _rowname) (Utilities|Struct|BreakDataTableRowHandle (Variables|Default|GetSpawnData)))
  ; Load the data layers before we activate them to ensure they're ready. ThrustersOut is activated by default so it does not need to be loaded.
  (bind _returnvalue (Utilities|GetDataTableRow _datatable _rowname))
  (bind _returnvalue_2 (WorldPartition|GetDataLayerManager))
  (Utilities|IsValid _returnvalue_2
    (:Is%20Valid
      (DataLayers|SetDataLayerRuntimeState _returnvalue_2 (Utilities|Select :Index (Variables|Default|GetSpawnedArmoryVariant) :NewEnumerator0 (.BarnTowelVariantInteriorDataLayer_32_D4AB6F074635E5252D0B7CA2EF0B7CB7 (Utilities|Struct|BreakBarnTowelArmorySpawnLocationInfo _returnvalue)) :NewEnumerator2 (.BarnTowelVariantInteriorDataLayer_32_D4AB6F074635E5252D0B7CA2EF0B7CB7 (Utilities|Struct|BreakBarnTowelArmorySpawnLocationInfo _returnvalue)) :NewEnumerator1 (.MurkyVoleVariantInteriorDataLayer_33_46056A00488B3EEE41DA50B6F7AF4626 (Utilities|Struct|BreakBarnTowelArmorySpawnLocationInfo _returnvalue)) :NewEnumerator3 (.BarnTowelVariantInteriorDataLayer_32_D4AB6F074635E5252D0B7CA2EF0B7CB7 (Utilities|Struct|BreakBarnTowelArmorySpawnLocationInfo _returnvalue))) "Loaded" false :!class /Script/Engine.DataLayerManager)
      (DataLayers|SetDataLayerRuntimeState _returnvalue_2 (Utilities|Select :Index (Variables|Default|GetSpawnedArmoryVariant) :NewEnumerator0 (.BarnTowelVariantLandingINDataLayer_35_2FF476334D20D0D7332405AC4D6D5834 (Utilities|Struct|BreakBarnTowelArmorySpawnLocationInfo _returnvalue)) :NewEnumerator2 (.BarnTowelVariantLandingINDataLayer_35_2FF476334D20D0D7332405AC4D6D5834 (Utilities|Struct|BreakBarnTowelArmorySpawnLocationInfo _returnvalue)) :NewEnumerator1 (.MurkyVoleVariantLandingINDataLayer_34_7F5618204A5C82CE07D6119B18613019 (Utilities|Struct|BreakBarnTowelArmorySpawnLocationInfo _returnvalue)) :NewEnumerator3 (.BarnTowelVariantLandingINDataLayer_35_2FF476334D20D0D7332405AC4D6D5834 (Utilities|Struct|BreakBarnTowelArmorySpawnLocationInfo _returnvalue))) "Loaded" false :!class /Script/Engine.DataLayerManager))
    (:Is%20Not%20Valid
      (Fortnite|FortShippingLog self "BARNTOWELARMORY: Unable to find Data Layer - the landing pad swap will not occur! This is unexpected except in TestLevels." true))))

(event Custom|OnArmoryStateUpdated
  (params (NewArmoryState :type Enum_BarnTowelArmory_LandingState%20Enum :default "NewEnumerator0"))
  (bind (_returnvalue_x _returnvalue_y _returnvalue_z) (Math|Vector|BreakVector (Transformation|GetActorLocation (Variables|Default|GetSpawnedArmoryPod))))
  ;[0.02214 0.342 0.1684] OnArmoryStateUpdated
  ; Run Armory spawning logic and send state-based events to reactive props
  (Utilities|FlowControl|Sequence
    (:then_0
      (switch Utilities|FlowControl|Switch|SwitchonEnum_BarnTowelArmory_LandingState NewArmoryState
        (:NewEnumerator1
          (CallFunction|LoadDataLayers))
        (:NewEnumerator3
          (CallFunction|SetLandinginOutDataLayers)
          (CallFunction|ActivateKnockbackSphere))
        (:NewEnumerator4
          (CallFunction|SpawnLandedArmoryVariant)
          (bind _returnvalue (Game|SpawnActorfromClass "/BarnTowelArmoryGameplay/Items/Environmental/Armory/BGA_BarnTowelArmory_Marker.BGA_BarnTowelArmory_Marker_C" (Math|Transform|MakeTransform (Math|Vector|MakeVector _returnvalue_x _returnvalue_y _returnvalue_z)) "Undefined" "MultiplyWithRoot" self (Variables|Default|GetSpawnedArmoryVariant) 0 0))
          (Variables|Default|SetArmoryMarker _returnvalue))
        (:NewEnumerator5
          (CallFunction|EndKnockbackSphere) ; knockback must last until all animations are completed
        )))
    (:then_1
      (for _array_element (Variables|Default|GetRegisteredArmoryComponents)
        (CallFunction|SendReactiveArmoryEvent _array_element NewArmoryState)))))

(event Custom|ActivateKnockbackSphere
  (bind _landingknockbackcapsule (Variables|Default|GetLandingKnockbackCapsule))
  ;[0.3812 0.6277 1] ActivateKnockbackSphere
  ; Players and Vehicles need to be knocked away from the Armory as it animates, because collision is not animated and will snap into existence when completed
  ;;[0.08351 1 0.1068] Activate collision on the knockback sphere
  (Networking|SwitchHasAuthority
    (:Authority
      (Variables|/Script/Engine.PrimitiveComponent|SetbGenerateOverlapEvents true _landingknockbackcapsule)
      (Collision|SetCollisionEnabled _landingknockbackcapsule "QueryOnly")
      ;;[0.3812 0.6277 1] Knockback any players already within the radius - if they are standing still, they are not detected by the overlap
      (bind _outactors (Collision|CapsuleOverlapActors (Transformation|GetActorLocation) (ScalableFloat|GetValueatLevel (Variables|Landing|Knockback|GetHFRow_KnockbackRadius)) (Variables|/Script/Engine.CapsuleComponent|GetCapsuleHalfHeight (Variables|Default|GetLandingKnockbackCapsule)) (Utilities|Array|MakeArray "ObjectTypeQuery3" "ObjectTypeQuery5")))
      (for _array_element _outactors
        (CallFunction|KnockbackPlayersorVehicles _array_element)))))

(event Custom|EndKnockbackSphere
  (bind _landingknockbackcapsule (Variables|Default|GetLandingKnockbackCapsule))
  ;[1 0.3436 0] EndKnockbackSphere
  ; Disable knockback area when Armory animations are completed
  (Variables|/Script/Engine.PrimitiveComponent|SetbGenerateOverlapEvents false _landingknockbackcapsule)
  (Collision|SetCollisionEnabled _landingknockbackcapsule "NoCollision"))

(event OnComponentBeginOverlap(LandingKnockbackCapsule)
  (params (OverlappedComponent :type Primitive%20Component%20Object%20Reference) (OtherActor :type Actor%20Object%20Reference) (OtherComp :type Primitive%20Component%20Object%20Reference) (OtherBodyIndex :type Integer) (bFromSweep :type Boolean) (SweepResult :type Hit%20Result%20Structure%20%28by%20ref%29))
  ;[0.4184 0.2549 0.1765] Knockback players and vehicles in the overlap area
  (CallFunction|KnockbackPlayersorVehicles OtherActor))

(event Custom|SetLandingInOutDataLayers
  (bind (_datatable _rowname) (Utilities|Struct|BreakDataTableRowHandle (Variables|Default|GetSpawnData)))
  ;[0.3694 0.6763 1] SetLandingInOutDataLayers
  ; Load the animated platform and hide the static flat platform
  (bind _returnvalue (Utilities|GetDataTableRow _datatable _rowname))
  (bind _returnvalue_2 (WorldPartition|GetDataLayerManager))
  (Utilities|IsValid _returnvalue_2
    (:Is%20Valid
      (DataLayers|SetDataLayerRuntimeState _returnvalue_2 (Utilities|Select :Index (Variables|Default|GetSpawnedArmoryVariant) :NewEnumerator0 (.BarnTowelVariantLandingINDataLayer_35_2FF476334D20D0D7332405AC4D6D5834 (Utilities|Struct|BreakBarnTowelArmorySpawnLocationInfo _returnvalue)) :NewEnumerator2 (.BarnTowelVariantLandingINDataLayer_35_2FF476334D20D0D7332405AC4D6D5834 (Utilities|Struct|BreakBarnTowelArmorySpawnLocationInfo _returnvalue)) :NewEnumerator1 (.MurkyVoleVariantLandingINDataLayer_34_7F5618204A5C82CE07D6119B18613019 (Utilities|Struct|BreakBarnTowelArmorySpawnLocationInfo _returnvalue)) :NewEnumerator3 (.BarnTowelVariantLandingINDataLayer_35_2FF476334D20D0D7332405AC4D6D5834 (Utilities|Struct|BreakBarnTowelArmorySpawnLocationInfo _returnvalue))) "Activated" false :!class /Script/Engine.DataLayerManager)
      (DataLayers|SetDataLayerRuntimeState _returnvalue_2 (.ArmorySharedLandingOUTDataLayer_38_6141E3FB4248A1D6BFE172BABD7E1238 (Utilities|Struct|BreakBarnTowelArmorySpawnLocationInfo _returnvalue)) "Unloaded" false :!class /Script/Engine.DataLayerManager))
    (:Is%20Not%20Valid
      (Fortnite|FortShippingLog self "BARNTOWELARMORY: Unable to find Data Layer - the landing pad swap will not occur! This is unexpected except in TestLevels." true))))

(event Custom|SpawnLandedArmoryVariant
  (bind (_datatable _rowname) (Utilities|Struct|BreakDataTableRowHandle (Variables|Default|GetSpawnData)))
  (bind _spawnedarmoryvariant (Variables|Default|GetSpawnedArmoryVariant))
  ;[0.4184 0.2549 0.1765] SpawnLandedArmoryVariant
  ; Spawn the landed armory level design data layer based on the current selected variant
  (bind _returnvalue (Utilities|GetDataTableRow _datatable _rowname))
  (bind _returnvalue_2 (WorldPartition|GetDataLayerManager))
  (Utilities|IsValid _returnvalue_2
    (:Is%20Valid
      (DataLayers|SetDataLayerRuntimeState _returnvalue_2 (Utilities|Select :Index _spawnedarmoryvariant :NewEnumerator0 (.BarnTowelVariantInteriorDataLayer_32_D4AB6F074635E5252D0B7CA2EF0B7CB7 (Utilities|Struct|BreakBarnTowelArmorySpawnLocationInfo _returnvalue)) :NewEnumerator2 (.BarnTowelVariantInteriorDataLayer_32_D4AB6F074635E5252D0B7CA2EF0B7CB7 (Utilities|Struct|BreakBarnTowelArmorySpawnLocationInfo _returnvalue)) :NewEnumerator1 (.MurkyVoleVariantInteriorDataLayer_33_46056A00488B3EEE41DA50B6F7AF4626 (Utilities|Struct|BreakBarnTowelArmorySpawnLocationInfo _returnvalue)) :NewEnumerator3 (.BarnTowelVariantInteriorDataLayer_32_D4AB6F074635E5252D0B7CA2EF0B7CB7 (Utilities|Struct|BreakBarnTowelArmorySpawnLocationInfo _returnvalue))) "Activated" false :!class /Script/Engine.DataLayerManager)
      ;[0.15 0.08931 0.03571] Handle the removal of the Skydrop Armory
      ; ON SERVER: Set the armory to be destroyed
      ; ON CLIENT: Set up a check to hide the armory only after the data layer is fully loaded to prevent visual popping
      (Networking|SwitchHasAuthority
        (:Authority
          (Class|BGAArmorySkydrop|DestroyLandedArmoryServer (Variables|Default|GetSpawnedArmoryPod)))
        (:Remote
          (Variables|Default|SetCachedQuerySources (Utilities|Array|MakeArray (Utilities|Struct|MakeWorldPartitionStreamingQuerySource (Transformation|GetActorLocation) (Variables|Default|GetSpatialQueryRadius) false (Utilities|Array|MakeArray (Utilities|String|StringToName (Utilities|GetObjectName (DataLayers|GetDataLayerInstancefromAsset _returnvalue_2 (Utilities|Select :Index _spawnedarmoryvariant :NewEnumerator0 (.BarnTowelVariantInteriorDataLayer_32_D4AB6F074635E5252D0B7CA2EF0B7CB7 (Utilities|Struct|BreakBarnTowelArmorySpawnLocationInfo _returnvalue)) :NewEnumerator2 (.BarnTowelVariantInteriorDataLayer_32_D4AB6F074635E5252D0B7CA2EF0B7CB7 (Utilities|Struct|BreakBarnTowelArmorySpawnLocationInfo _returnvalue)) :NewEnumerator1 (.MurkyVoleVariantInteriorDataLayer_33_46056A00488B3EEE41DA50B6F7AF4626 (Utilities|Struct|BreakBarnTowelArmorySpawnLocationInfo _returnvalue)) :NewEnumerator3 (.BarnTowelVariantInteriorDataLayer_32_D4AB6F074635E5252D0B7CA2EF0B7CB7 (Utilities|Struct|BreakBarnTowelArmorySpawnLocationInfo _returnvalue))))))) true)))
          (CallFunction|CheckExteriorLandedStreamingComplete))))
    (:Is%20Not%20Valid
      (Fortnite|FortShippingLog self "BARNTOWELARMORY: Unable to find Data Layer - the landing pad swap will not occur! This is unexpected except in TestLevels." true))))

(event EventEndPlay
  (params (EndPlayReason :type EEndPlayReason%20Enum :default "Destroyed"))
  ;[1 0.3815 0] EndPlay
  ; BP Cleanup; clear timers, unbind events, end async tasks, etc.
  (Utilities|FlowControl|Sequence
    (:then_0
      (bind _spawnedarmorypod (Variables|Default|GetSpawnedArmoryPod)
        (:then
          (Default|UnbindEventfromOnArmoryStateChanged _spawnedarmorypod (EventDispatchers|CreateEvent :!event Custom|OnArmoryStateUpdated)))
        (:else)))
    (:then_1
      (|CallParentFunction EndPlayReason) ; run parent end play last in case \nit needs references to something \nthat is cleaned up
    )))

(fn CacheReactiveArmoryPropComponent
  (params (RegistrationTag :type Gameplay%20Tag%20Structure) (RegisteredActor :type Actor%20Object%20Reference))
  (bind _returnvalue (Actor|GetComponentbyClass RegisteredActor "/BarnTowelArmoryGameplay/Items/BP_BarnTowelArmory_ReactiveProp_Component.BP_BarnTowelArmory_ReactiveProp_Component_C"))
  ;[0.4343 0.7035 1] CacheReactiveArmroyPropComponent
  ; Store any unique reactive props that will need to have behavior when the armory is landed or opened
  (Utilities|IsValid _returnvalue
    (:Is%20Valid
      (Utilities|Array|AddUnique (Variables|Default|GetRegisteredArmoryComponents) _returnvalue)
      (CallFunction|SendReactiveArmoryEvent _returnvalue (Variables|Default|GetCurrentArmoryGameplayState)) ; send an armory event since it may have missed one\nif it is registered late
      (Variables|/BarnTowelArmoryGameplay/Items/BP_BarnTowelArmory_ReactiveProp_Component.BP_BarnTowelArmory_ReactiveProp_Component_C|SetOwningArmorySpawnManager self _returnvalue) ; send reference to manager for any \ncross-communication events (opening, opened, etc.)
    )))

(fn CheckExteriorLandedStreamingComplete
  (bind _returnvalue (WorldSubsystems|GetWorldPartitionSubsystem))
  ;[0.2783 0.4748 0.7474] CheckExteriorLandedStreamingComplete
  ; Once the Armory data layer has fully loaded on the Client, then hide the Skyfall armory
  (Networking|CanExecuteCosmeticEvents
    (:True
      (bind _spawnedarmorypod (Variables|Default|GetSpawnedArmoryPod)
        (:then
          (if (not (Variables|Default|GetbStreamingComplete))
            (Utilities|IsValid _returnvalue
              (:Is%20Valid
                (if (Streaming|IsStreamingCompleted _returnvalue "Activated" (Variables|Default|GetCachedQuerySources) true :!class /Script/Engine.WorldPartitionSubsystem)
                  (Variables|Default|SetbStreamingComplete true)
                  (Class|BGAArmorySkydrop|HideLandedArmoryClient _spawnedarmorypod) ; hides the armory on the client
                  (else
                    (bind _returnvalue_2 (Utilities|Time|SetTimerforNextTickbyEvent (EventDispatchers|CreateEvent :!event CheckExteriorLandedStreamingComplete)))
                    (Variables|Default|SetTimerHandle_CheckExteriorLoaded _returnvalue_2)))))))
        (:else)) ; check validity early - no need to run \nlogic if there's no Armory to hide
    )))

(fn GetAndRegisterReactiveArmoryProps
  (bind (_datatable _rowname) (Utilities|Struct|BreakDataTableRowHandle (Variables|Default|GetSpawnData)))
  (bind _returnvalue (WorldSubsystems|GetFortGlobalActorRegistrySubsystem))
  ;[0.4973 0.715 1] GetAndRegisterLandingLocations
  ; Find all the landing locations for this manager to later decide what to activate
  (bind _returnvalue_2 (Utilities|GetDataTableRow _datatable _rowname))
  (bind _outactors (Class|FortGlobalActorRegistrySubsystem|GetRegisteredActors _returnvalue (.RegistrationTag_7_59CAA6B3460D7EAAD30B5E93FB994000 (Utilities|Struct|BreakBarnTowelArmorySpawnLocationInfo _returnvalue_2)))) ; get all the current registered actors
  (for _array_element _outactors
    ;;[0.4973 0.715 1] Add all found actors - run additional validation if necessary
    (CallFunction|CacheReactiveArmoryPropComponent (.RegistrationTag_7_59CAA6B3460D7EAAD30B5E93FB994000 (Utilities|Struct|BreakBarnTowelArmorySpawnLocationInfo _returnvalue_2)) _array_element))
  ;;[0.4973 0.715 1] Bind to Data Registry object - Will register any defenses that are added to the registry object after the initial run
  (bind _returnvalue_3 (Class|FortGlobalActorRegistrySubsystem|GetActorRegistryDataforGameplayTag _returnvalue (.RegistrationTag_7_59CAA6B3460D7EAAD30B5E93FB994000 (Utilities|Struct|BreakBarnTowelArmorySpawnLocationInfo _returnvalue_2))))
  (Utilities|IsValid _returnvalue_3
    (:Is%20Valid
      (bind _output_get (Variables|Default|SetCachedReactiveArmoryPropRegistrySubsystem _returnvalue_3))
      (EventDispatchers|BindEventtoOnActorRegistered _output_get (EventDispatchers|CreateEvent :!event CacheReactiveArmoryPropComponent)))))

(fn KnockbackPlayersOrVehicles
  (params (ActorToLaunch :type Actor%20Object%20Reference))
  (bind _self self)
  (bind _returnvalue (Ability|GetAbilitySystemComponent _self))
  ;[0.4308 0.6721 1] Launch characters and vehicles if we run into them
  (Networking|SwitchHasAuthority
    (:Authority
      (Utilities|IsValid _returnvalue
        (:Is%20Valid
          (GameplayEffects|ApplyGameplayEffectToTarget _returnvalue (Variables|Landing|Knockback|GetGE_Knockback) (Ability|GetAbilitySystemComponent ActorToLaunch) 0.0 (Class|AthenaFunctionLibrary|MakeSimpleFortEffectContext self self)))))))

(fn PlayerEntered
  (params (FortPawn :type Fort%20Pawn%20Object%20Reference))
  (for _array_element (Variables|Default|GetRegisteredArmoryComponents)
    (Class|BPBarnTowelArmoryReactivePropComponent|OnArmoryEntered _array_element FortPawn)
    (Actor|DestroyActor (Variables|Default|GetArmoryMarker))))

(fn SendReactiveArmoryEvent
  (params (ReactiveProp :type BP%20Barn%20Towel%20Armory%20Reactive%20Prop%20Component%20Object%20Reference) (ArmoryState :type Enum_BarnTowelArmory_LandingState%20Enum :default "NewEnumerator0"))
  (bind _spawnedarmoryvariant (Variables|Default|GetSpawnedArmoryVariant))
  (switch Utilities|FlowControl|Switch|SwitchonEnum_BarnTowelArmory_LandingState ArmoryState
    (:NewEnumerator1
      (Class|BPBarnTowelArmoryReactivePropComponent|OnArmorySpawnedReceived ReactiveProp _spawnedarmoryvariant))
    (:NewEnumerator2
      (Class|BPBarnTowelArmoryReactivePropComponent|OnArmoryWingsDeployedReceived ReactiveProp))
    (:NewEnumerator3
      (Class|BPBarnTowelArmoryReactivePropComponent|OnArmoryThrustersActivatedReceived ReactiveProp))
    (:NewEnumerator4
      (Class|BPBarnTowelArmoryReactivePropComponent|OnArmoryLandedReceived ReactiveProp _spawnedarmoryvariant))
    (:NewEnumerator5
      (Class|BPBarnTowelArmoryReactivePropComponent|OnArmoryOpeningReceived ReactiveProp))
    (:NewEnumerator6
      (Class|BPBarnTowelArmoryReactivePropComponent|OnArmoryOpenReceived ReactiveProp))))

(fn SetNewArmoryState
  (params (NewArmoryState :type Enum_BarnTowelArmory_LandingState%20Enum :default "NewEnumerator0"))
  (bind _output_get (Variables|Default|SetCurrentArmoryGameplayState NewArmoryState))
  (CallFunction|OnArmoryStateUpdated _output_get) ; calling this manually because the Bind seems to miss the BeginPlay( ) event
)

(fn SpawnArmoryInAir
  (params (SelectedArmoryVariant :type Enum_BarnTowelArmory_SelectedVariant%20Enum :default "NewEnumerator0"))
  (bind (_datatable _rowname) (Utilities|Struct|BreakDataTableRowHandle (Variables|Default|GetSpawnData)))
  (bind _returnvalue (Transformation|GetActorRotation))
  (bind _returnvalue_2 (Transformation|GetActorLocation))
  (bind _hfrow_spawnheightabovelandingplatform (Variables|Default|GetHFRow_SpawnHeightAboveLandingPlatform))
  (bind _returnvalue_3 (ScalableFloat|GetValueatLevel _hfrow_spawnheightabovelandingplatform))
  (bind _returnvalue_4 (Math|Vector|MakeVector 0.0 0.0 _returnvalue_3))
  (bind _returnvalue_5 (+ _returnvalue_2 _returnvalue_4))
  (bind _returnvalue_6 (Math|Transform|MakeTransform _returnvalue_5 _returnvalue))
  ;[0 0.07435 1] SpawnArmoryInAir
  ; Spawn the Armory pod in the air above this landing pad, if it is enabled. Then setup landed and opened events for that spawned armory
  (bind _output_get (Variables|Default|SetSpawnedArmoryVariant SelectedArmoryVariant))
  (Utilities|FlowControl|Sequence
    (:then_0
      (bind _returnvalue_7 (Utilities|GetDataTableRow _datatable _rowname)
        (:then
          (if (.bIsEnabled_1_D425057A4FB4779E062A6E9BFEE5BCF8 (Utilities|Struct|BreakBarnTowelArmorySpawnLocationInfo _returnvalue_7))
            (switch Utilities|FlowControl|Switch|SwitchonEnum_BarnTowelArmory_SelectedVariant _output_get
              (:NewEnumerator0
                _)
              (:NewEnumerator2
                _)
              (:NewEnumerator1
                (bind _returnvalue_8 (Building|SpawnBuildingGameplayActorfromWorldContext "/BarnTowelArmoryGameplay/Items/Environmental/Armory/BGA_Armory_Skydrop_MurkyVole.BGA_Armory_Skydrop_MurkyVole_C" _returnvalue_6))
                (bind _asbga_armory_skydrop (Utilities|Casting|CastToBGA_Armory_Skydrop _returnvalue_8))
                (bind _output_get_2 (Variables|Default|SetSpawnedArmoryPod _asbga_armory_skydrop))
                (Variables|/BarnTowelArmoryGameplay/Items/Environmental/Armory/BGA_Armory_Skydrop.BGA_Armory_Skydrop_C|SetLandingPosition _returnvalue_2 _output_get_2)
                (Default|BindEventtoOnArmoryStateChanged (Variables|Default|GetSpawnedArmoryPod) (EventDispatchers|CreateEvent :!event SetNewArmoryState))
                (CallFunction|SetNewArmoryState "NewEnumerator1")
                (return (Variables|Default|GetSpawnedArmoryPod)))
              (:NewEnumerator3
                _))
            (bind _returnvalue_9 (Building|SpawnBuildingGameplayActorfromWorldContext "/BarnTowelArmoryGameplay/Items/Environmental/Armory/BGA_Armory_Skydrop.BGA_Armory_Skydrop_C" _returnvalue_6))
            (bind _asbga_armory_skydrop_2 (Utilities|Casting|CastToBGA_Armory_Skydrop _returnvalue_9))
            (bind _output_get_3 (Variables|Default|SetSpawnedArmoryPod _asbga_armory_skydrop_2))
            (Variables|/BarnTowelArmoryGameplay/Items/Environmental/Armory/BGA_Armory_Skydrop.BGA_Armory_Skydrop_C|SetLandingPosition _returnvalue_2 _output_get_3)
            (else
              (Fortnite|FortShippingLog self "BARNTOWELARMORY: Spawner attempted to use a location that is not enabled. Should not reach this point in logic if the spawner is disabled!" true))))
        (:RowNotFound
          (Fortnite|FortShippingLog self "BARNTOWELARMORY: Spawner attempted to create armory, but did not have valid Spawn Data!" true))))
    (:then_1
      (return 0))))

(fn ConstructionScript)
